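; Skeletal mesh mannequin actor for Season 36 Battle Pass quests that can play a requested montage and tunes its meshes for performance.
; A custom event async-loads a soft montage, enables custom depth on the Face part, then loops the montage on the skeletal mesh.
; After reinitialization, every skeletal mesh gets custom depth, render-only pose ticking and a minimum LOD based on the cached effects-quality setting.
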(event Custom|LoadAndPlayMontage
  (params (LocalMontage :type Anim%20Montage%20Soft%20Object%20Reference))
  (bind _object (Utilities|AsyncLoadAsset LocalMontage)
    (:Completed
      (Rendering|SetRenderCustomDepth (Customization|GetPartSkeletalMeshComponent "Face") true)
      (bind _asanim_montage (Utilities|Casting|CastToAnimMontage _object))
      (Utilities|IsValid _asanim_montage
        (:Is%20Valid
          (bind _output_get (Variables|Default|SetMontageActual _asanim_montage))
          (bind _skeletalmeshcomponent (Variables|SkeletalMeshActor|GetSkeletalMeshComponent)
            (:then
              (Components|Animation|PlayAnimation _skeletalmeshcomponent _output_get true))
            (:else)))))))

(event EventHandleReInitializeComplete
  (CallFunction|ValidateMeshSettings))

(fn GetFXQuality
  (if (> (Variables|Default|GetCurrentFXQuality) -1)
    (return (Variables|Default|GetCurrentFXQuality))
    (else
      (bind _returnvalue (Development|GetConsoleVariableIntValue "sg.EffectsQuality"))
      (bind _output_get (Variables|Default|SetCurrentFXQuality _returnvalue))
      (return _output_get))))

(fn ConstructionScript)

(fn ValidateMeshSettings
  (bind _quality (CallFunction|GetFXQuality))
  (for _array_element (Actor|GetComponentsByClass "/Script/Engine.SkeletalMeshComponent")
    (Components|SkinnedMesh|OverrideMinLOD _array_element (select (< _quality 2) 2 1))
    (Rendering|SetRenderCustomDepth _array_element true)
    (Variables|/Script/Engine.SkinnedMeshComponent|SetVisibilityBasedAnimTickOption "OnlyTickPoseWhenRendered" _array_element)))
