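; Mist Raven proxy actor for Season 36 story quests, with a mannequin, Niagara effect and unstable wraith post process.
; Construction and BeginPlay each create the post process MID; only construction sets its Position, and BeginPlay binds Niagara to the mannequin mesh.
; Sphere overlaps play and reverse a timeline that fades the material's Blend.
; Gameplay-tagged side effects, handled only without authority, play the instance's mannequin montage or spawn poof VFX.
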(event Collision|EventActorBeginOverlap
  (params (OtherActor :type Actor%20Object%20Reference)))

(event EventBeginPlay
  ;[0.15 0.07098 0.1443 0.5] Post Process Setup
  (|CallParentFunction)
  (bind _returnvalue (Rendering|Material|CreateDynamicMaterialInstance "/BattlePassS36_Quests/Items/QuestActors/Story/P01/PPMI_MistRaven_UnstableWraithForm01.PPMI_MistRaven_UnstableWraithForm01" "None" "None" :!class /Script/Engine.KismetMaterialLibrary))
  (Variables|Default|SetMID _returnvalue)
  (Variables|/Script/Engine.PostProcessComponent|SetSettings (Utilities|Struct|MakePostProcessSettings (Utilities|Struct|MakeWeightedBlendables (Utilities|Array|MakeArray (Utilities|Struct|MakeWeightedBlendable 1.0 (Variables|Default|GetMID))))) (Variables|Default|GetPostProcess))
  (bind _asfort_player_mannequin (Utilities|Casting|CastToFortPlayerMannequin (Variables|/Script/Engine.ChildActorComponent|GetChildActor (Variables|Default|GetMannequin%20Actor))))
  (Niagara|SetNiagaraSkeletalMeshComponent (Variables|Default|GetNiagara) "" (Variables|/Script/Engine.SkeletalMeshActor|GetSkeletalMeshComponent _asfort_player_mannequin)))

(event EventTick
  (params (DeltaSeconds :type Float%20%28single-precision%29)))

(event SideEffects|EventBP_ApplySideEffect
  (params (ReactionSource :type Object%20Reference) (TargetObject :type Object%20Reference) (SideEffectTriggerTag :type Gameplay%20Tag%20Structure%20%28by%20ref%29) (SideEffectTriggerType :type EBlueprintableSideEffectTriggerType%20Enum%20%28by%20ref%29 :default "TriggeredAsTarget"))
  ;[0.04564 0.05241 0.15 0.5] Mannequin Animation
  (Networking|SwitchHasAuthority
    (:Remote
      (switch Utilities|FlowControl|Switch|SwitchonGameplayTag SideEffectTriggerTag
        (:Quests.S36.SDA.BPSideEffect.A
          (bind _asbp_s36_mannequin (Utilities|Casting|CastToBP_S36_Mannequin (Variables|/Script/Engine.ChildActorComponent|GetChildActor (Variables|Default|GetMannequin%20Actor))))
          (Class|BPS36Mannequin|LoadandPlayMontage _asbp_s36_mannequin (Variables|Default|GetInstanceMontage)))
        (:Quests.S36.SDA.BPSideEffect.B)
        (:Quests.S36.SDA.BPSideEffect.C
          ; POOF FX
          (Niagara|SpawnSystemAttached "/BattlePassS36_Quests/VFX/NS_MistRaven_Poof.NS_MistRaven_Poof" (Variables|Default|GetMannequin%20Actor)))))))

(event OnComponentBeginOverlap(SphereforPostProcessEffect)
  (params (OverlappedComponent :type Primitive%20Component%20Object%20Reference) (OtherActor :type Actor%20Object%20Reference) (OtherComp :type Primitive%20Component%20Object%20Reference) (OtherBodyIndex :type Integer) (bFromSweep :type Boolean) (SweepResult :type Hit%20Result%20Structure%20%28by%20ref%29))
  ;[0.15 0.04215 0.1455 0.5] Post Process Fading
  (bind _asfort_player_pawn (Utilities|Casting|CastToFortPlayerPawn OtherActor))
  (Variables|Default|SetPlayerPawn _asfort_player_pawn)
  (bind (_direction _blend) (Timeline|Timeline_0 PlayFromStart :NewTime 1.0)
    (:Update
      (Rendering|Material|SetScalarParameterValue (Variables|Default|GetMID) "Blend" _blend :!class /Script/Engine.MaterialInstanceDynamic))))

(event OnComponentEndOverlap(SphereforPostProcessEffect)
  (params (OverlappedComponent :type Primitive%20Component%20Object%20Reference) (OtherActor :type Actor%20Object%20Reference) (OtherComp :type Primitive%20Component%20Object%20Reference) (OtherBodyIndex :type Integer))
  ;[0.15 0.04215 0.1455 0.5] Post Process Fading
  (Timeline|Timeline_0 ReverseFromEnd))

(fn ConstructionScript
  (bind _returnvalue (Rendering|Material|CreateDynamicMaterialInstance "/BattlePassS36_Quests/Items/QuestActors/Story/P01/PPMI_MistRaven_UnstableWraithForm01.PPMI_MistRaven_UnstableWraithForm01" "None" "None" :!class /Script/Engine.KismetMaterialLibrary))
  (Variables|Default|SetMID _returnvalue)
  (Rendering|Material|SetVectorParameterValue (Variables|Default|GetMID) "Position" (Math|Conversions|ToLinearColor(Vector) (Variables|/Script/Engine.SceneComponent|GetRelativeLocation (Variables|Default|GetSphere%20for%20Post%20Process%20Effect))) :!class /Script/Engine.MaterialInstanceDynamic)
  (Variables|/Script/Engine.PostProcessComponent|SetSettings (Utilities|Struct|MakePostProcessSettings (Utilities|Struct|MakeWeightedBlendables (Utilities|Array|MakeArray (Utilities|Struct|MakeWeightedBlendable 1.0 (Variables|Default|GetMID))))) (Variables|Default|GetPostProcess)))
