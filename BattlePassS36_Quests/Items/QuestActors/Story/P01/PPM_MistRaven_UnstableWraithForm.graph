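# Post-process material for the Battle Pass S36 Mist Raven quest's unstable wraith form, faded over the scene by a Blend parameter.
# Two curl-noise-warped polar-coordinate layers, offset half a period, form the TendrilsMask for tinted radial tendrils with eye-adaptation compensation.
# Scene color is also UV-distorted and desaturated toward the left and right edges.
# Quality and low/mobile switches simplify the effect; a camera-distance fade is left unconnected.
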Material /BattlePassS36_Quests/Items/QuestActors/Story/P01/PPM_MistRaven_UnstableWraithForm
graph {
 properties {blendableLocation=BL_SceneColorBeforeBloom, materialDomain=MD_PostProcess, usedShadingModels=Default Lit}
 nodes [
  n1 = TextureCoordinate()
  n2 = MaterialFunctionCall(UVs=@n1, Texture Scale=Constant() {r=1.5}) {materialFunction=&ScaleUVsByCenter}
  n3 = Multiply(A=Multiply(A=Add(A=@n2.UVs) {constB=-0.5}) {constB=2}, B=Power(Base=NamedRerouteUsage()) {constExponent=4})
  n4 = QualitySwitch(Default=Add(A=@n1, B=Multiply(A=@n3) {constB=-0.1}), Low=TextureCoordinate())
  n5 = ComponentMask(None=SceneTexture(UVs=@n4) {bFiltered=true, sceneTextureId=PPI_PostProcessInput0}.Color) {b=true, g=true, r=true}
  n6 = Desaturation(None=ComponentMask(None=SceneTexture(UVs=@n4) {bFiltered=true, sceneTextureId=PPI_PostProcessInput0}.Color) {b=true, g=true, r=true})
  n7 = ScreenPosition()
  n8 = Abs(None=Multiply(A=Add(A=@n7.ViewportUV) {constB=-0.5}) {constB=2})
  n9 = MaterialFunctionCall(Float2=Power(Base=@n8, Exp=ScalarParameter() {defaultValue=2, parameterName=Desaturation_Exp, sliderMax=16})) {materialFunction=&BreakOutFloat2Components}
  n10 = Divide(A=Add(A=@n9.R, B=@n9.R)) {constB=1.5}
  n11 = LinearInterpolate(A=@n5, B=@n6, Alpha=Saturate(None=Power(Base=@n10) {constExponent=3})) {constAlpha=1}
  n12 = Constant3Vector() {constant=<rgba 0.35 0.35 0.35 0>}
  n13 = Power(Base=NamedRerouteUsage())
  n14 = LinearInterpolate(
   A=VectorParameter() {parameterName=Tendril_TintColor, defaultValue=<rgba 0.2258 0.0557 0.3906 1>}.RGB
   B=MaterialFunctionCall!(Default=Multiply(A=@n6, B=@n12), Low/Mobile=Constant3Vector()) {materialFunction=/Game/Effects/Fort_Effects/Materials/_Functions/MF_LowMobile_PathingSwitch}.Result
   Alpha=Saturate(None=Add(A=@n13) {constB=-0.1})
  )
  n15 = MaterialFunctionCall(Float2=Power(Base=@n8, Exp=ScalarParameter() {defaultValue=4, parameterName=TendrilFrame_Exp, sliderMax=16}) {constExponent=4}) {materialFunction=&BreakOutFloat2Components}
  n16 = Divide(A=Add(A=@n15.R, B=@n15.R))
  n17 = Multiply(A=Power(Base=Saturate(None=@n16)) {constExponent=0.4}, B=Multiply(A=NamedRerouteUsage()) {constB=1.5})
  n18 = Constant() {r=0.15}
  LinearInterpolate(
   A=ComponentMask(None=SceneTexture() {bFiltered=true, sceneTextureId=PPI_PostProcessInput0}.Color) {b=true, g=true, r=true}
   B=LinearInterpolate(
    A=MaterialFunctionCall(Default=@n11, Low/Mobile=@n5) {materialFunction=&MF_LowMobile_PathingSwitch}.Result
    B=EyeAdaptationInverse(LightValueInput=@n14).EyeAdaptationInverse
    Alpha=EyeAdaptationInverse(LightValueInput=@n17, AlphaInput=@n18).EyeAdaptationInverse
   ) {constAlpha=1}
   Alpha=ScalarParameter() {parameterName=Blend, sliderMax=1}
  )
  n19 = Constant()
  n20 = Constant() {r=0.2}
  n21 = TextureCoordinate()
  n22 = Multiply(A=TextureCoordinate()) {constB=0.3}
  n23 = TextureSample(UVs=Panner(Coordinate=@n22) {speedY=-0.05}) {samplerType=SAMPLERTYPE_LinearColor, texture=/Game/Accessories/FORT_Backpacks/Backpack_M_MED_Carabus/FX/T_PerlinCurl}
  n24 = ScalarParameter() {defaultValue=0.15, parameterName=Deform_intensity, sliderMax=1}
  n25 = Multiply(A=Add(A=ComponentMask(None=@n23.RGB) {g=true, r=true}) {constB=-0.5}, B=@n24) {constB=0.06}
  n26 = Constant() {r=0.666}
  n27 = Divide(B=ScalarParameter() {defaultValue=4, parameterName=Tendril_Amount})
  n28 = MaterialFunctionCall(Float2=MaterialFunctionCall!(UV Coord=Add(A=@n21, B=@n25), LenghtScale=@n26, RadialScale=@n27) {materialFunction=/Game/Global/MaterialFunctions/TechArt/MF_PolarCoordinates}.Result) {materialFunction=&BreakOutFloat2Components}
  n29 = Frac(None=@n28.G)
  n30 = Multiply(A=@n28.R, B=Abs(None=Add(A=@n29) {constB=-0.5}))
  n31 = TextureSample(UVs=Panner(Coordinate=Multiply(A=@n22) {constB=0.666}) {speedY=0.05}) {samplerType=SAMPLERTYPE_LinearColor, texture=&T_PerlinCurl}
  n32 = Multiply(A=Add(A=ComponentMask(None=@n31.RGB) {g=true, r=true}) {constB=-0.5}, B=@n24) {constB=0.06}
  n33 = MaterialFunctionCall(Float2=MaterialFunctionCall(UV Coord=Add(A=@n21, B=@n32), LenghtScale=@n26, RadialScale=@n27) {materialFunction=&MF_PolarCoordinates}.Result) {materialFunction=&BreakOutFloat2Components}
  n34 = Frac(None=Add(A=@n33.G) {constB=0.5})
  n35 = Multiply(A=@n33.R, B=Abs(None=Add(A=@n34) {constB=-0.5}))
  NamedRerouteDeclaration(None=Max(A=SmoothStep(Min=@n19, Max=@n20, Value=@n30) {constMax=0.15}, B=SmoothStep(Min=@n19, Max=@n20, Value=@n35) {constMax=0.15})) {name=TendrilsMask, nodeColor=<rgba 0 0.4141 1 1>}
  n36 = Distance(A=VectorParameter() {parameterName=Position}.RGB, B=CameraPositionWS())
  OneMinus(None=Saturate(None=Divide(A=@n36) {constB=200}))
  TransformPosition(None=Constant3Vector() {constant=<rgba -944 5000 132 0>}) {transformType=TRANSFORMPOSSOURCE_World}
 ]
}
